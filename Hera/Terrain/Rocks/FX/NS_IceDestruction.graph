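# Niagara one-shot burst effect for breaking ice, throwing out mesh chunks, ice sprites, mist and small droplets.
# Shared system values size the spawn volumes, set most burst counts through curves and tint ice by time of day.
# Chunks have a sparser CPU LOW emitter and a GPU HIGH emitter with collision.
# A system line trace supplies a ground height that kills falling particles below it; two debug emitters are disabled.
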NiagaraSystem /Hera/Terrain/Rocks/FX/NS_IceDestruction
dependencies {
 usedDataInterfaces [
  &NiagaraDataInterfaceCurve
  &NiagaraDataInterfaceSpriteRendererInfo
  &NiagaraDataInterfaceCamera
  &NiagaraDataInterfaceCollisionQuery
 ]
 usedDynamicInputs [
  &Multiply_Vector
  &self:avg
  &MakeFloatFromVector
  &VectorFromFloat
 ]
 usedModules [
  &self:SystemSpawnScript.SystemScriptSource.SystemScriptGraph.NiagaraNodeAssignment_0.SetVariables_#1
  &self:LineTrace
  &SystemState
  &self:SystemSpawnScript.SystemScriptSource.SystemScriptGraph.NiagaraNodeAssignment_1.SetVariables_#2
 ]
}
emitters {
 Bounds_Debug {
  bEnabled=false
  simTarget=GPUComputeSim
  emitterUpdateScript [
   &EmitterState()
   &SpawnBurst_Instantaneous(Spawn Count=500, Use Spawn Probability=false, UseLoopCountLimit=false)
  ]
  particleSpawnScript [
   &InitializeParticle(
    Color Mode=::NewEnumerator0
    Lifetime=5
    Sprite Size Mode=::NewEnumerator3
    Uniform Sprite Size=20
    UsePositionOffset=false
    Write Material Random=true
   )
   &ShapeLocation(MaskSpawnGroup=false, Non Uniform Scale=$System.SysObjectScale, Offset Mode=::NewEnumerator0, Override Seed=false, UseOwnerScale=true) {module=ShapeLocation002}
   &ShapeLocation(MaskSpawnGroup=false, Non Uniform Scale=$System.SysObjectScale, Override Seed=false, UseOwnerScale=true) {enabled=false}
  ]
  particleUpdateScript [
   &ParticleState()
   &GravityForce(Gravity=&Multiply_VectorByFloat(Vector=<0 0 -980>, Float=1)) {enabled=false}
   &Drag(Drag=0, Ignore Mass=false, Rotational Drag=3, Use Linear Drag=true, Use Rotational Drag=true)
   &Collision(
    Age Colliding Particles=false
    Collision Enabled=&SetBoolByFloatComparison(A=$Particles.Lifetime, B=0.2)
    Control Roll On Collision=true
    Friction=0.6
    Friction During a Bounce=0.1
    GPU Collision Type=::NewEnumerator2
    KillOnCollision=false
    Mesh Dimensions:Vector3f=$Particles.Scale
    Module.Control Rest State Via Surface Normal=false
    Module.EnableMaxCollisionCount=false
    Module.Randomize Collision Normal Vector=false
    Radius Calculation Type=::NewEnumerator1
    Radius Scale for Rotation=30
    Restitution=0.3
    ScaleRateOfRotation=true
   ) {enabled=false, module=Collision001}
   &SolveForcesAndVelocity(Clamp Velocity=false, Limit Acceleration=false)
  ]
  rendererClasses [&NiagaraSpriteRendererProperties]
  renderers [{rendererClass=&NiagaraSpriteRendererProperties}]
 }
 Chunks_CPU_LOW {
  emitterUpdateScript [
   &EmitterState()
   &SpawnBurst_Instantaneous(
    Spawn Count=&MakeIntFromFloat(Float=&FloatFromCurve(
      FloatCurve=DataInterface() {propertyValues {Curve {keys [<at 2 30>,<at 7 100>,<at 19 200>]}}}
      CurveIndex=$System.SysSpawnCount
      Scale Curve=0.3
     ))
    Use Spawn Probability=false
    UseLoopCountLimit=false
   ) {module=SpawnBurst_Instantaneous002}
  ]
  particleSpawnScript [
   &InitializeParticle(
    AdjustAlpha=false
    AdjustHue=true
    AdjustSaturation=true
    AdjustValue=true
    Color=$System.IceColorTOD
    Color Mode=::NewEnumerator3
    Hue Shift Range:Vector2f=<-0.03 0.03>
    Lifetime Max=1.5
    Lifetime Min=0.8
    Lifetime Mode=::NewEnumerator1
    Mass Mode=::NewEnumerator1
    Mesh Scale Mode=::NewEnumerator3
    Mesh Uniform Scale=&Multiply_Float(A=$System.SysParticleScale, B=&RandomRangeFloat(Minimum=0.2, Maximum=0.4))
    Saturation Range:Vector2f=<0.6 1.2>
    UsePositionOffset=false
    Value Range:Vector2f=<0.8 1.2>
    Write Material Random=true
   ) {module=InitializeParticle001}
   &ShapeLocation(
    Apply Owner Scale=HlslExpression() {hlslExpression="abs(Engine.Owner.Scale)"}
    MaskSpawnGroup=false
    Non Uniform Scale=$System.SysObjectScale
    Offset Mode=::NewEnumerator0
    Override Seed=false
    UseOwnerScale=false
   ) {module=ShapeLocation002}
   &AddVelocity(
    Module.Use Falloff Distance=false
    Origin Offset=<0 0 -200>
    Origin Offset Coordinate Space=::World
    Override Seed=false
    Velocity Mode=::NewEnumerator1
    Velocity Speed=&RandomRangeFloat(Minimum=100, Maximum=600)
   )
   &InitialMeshOrientation(Use Rotation Vector=false)
   &MeshRotationForce(Rotation=&RandomRangeVector(Minimum=<-50 -50 -50>, Maximum=<50 50 50>))
   &ApplyInitialForces()
  ]
  particleUpdateScript [
   &ParticleState()
   &ScaleMeshSize(Scale Factor=&Multiply_VectorByFloat(
     Vector=<1 1 1>
     Float=&FloatFromCurve(
      FloatCurve=DataInterface() {propertyValues {Curve {keys [
          {arriveTangent=13.7848, interpMode=RCIM_Cubic, leaveTangent=13.7848, tangentMode=RCTM_User, time=0, value=0}
          <at 0.15 1>
          {interpMode=RCIM_Cubic, tangentMode=RCTM_User, time=0.8, value=1}
          {tangentMode=RCTM_User, time=1, value=0}
         ]}}}
      Scale Curve=0.75
     )
    ))
   &GravityForce(Gravity=&Multiply_VectorByFloat(Vector=<0 0 -980>, Float=1))
   &Drag(Drag=0, Ignore Mass=false, Rotational Drag=3, Use Linear Drag=true, Use Rotational Drag=true)
   &SolveForcesAndVelocity(Clamp Velocity=false, Limit Acceleration=false)
   &KillParticles(Kill Particles=HlslExpression() {hlslExpression="(Particles.Position.z < System.CollisionPosWorld.z) &&&& System.CollisionValid"})
   /SetVariables(Particles.Scale=HlslExpression() {hlslExpression="abs(Particles.Scale)"}) {isSetParameters=true}
  ]
  rendererClasses [&NiagaraMeshRendererProperties]
  renderers [{rendererClass=&NiagaraMeshRendererProperties, rendererIndex=0}]
 }
 Chunks_GPU_HIGH {
  simTarget=GPUComputeSim
  emitterUpdateScript [
   &EmitterState()
   &SpawnBurst_Instantaneous(
    Spawn Count=&MakeIntFromFloat(Float=&FloatFromCurve(FloatCurve=DataInterface() {propertyValues {Curve {keys [<at 2 30>,<at 7 100>,<at 19 200>]}}}, CurveIndex=$System.SysSpawnCount))
    Use Spawn Probability=false
    UseLoopCountLimit=false
   )
  ]
  particleSpawnScript [
   &InitializeParticle(
    AdjustAlpha=false
    AdjustHue=true
    AdjustSaturation=true
    AdjustValue=true
    Color=$System.IceColorTOD
    Color Mode=::NewEnumerator3
    Hue Shift Range:Vector2f=<-0.03 0.03>
    Lifetime Min=1.25
    Lifetime Mode=::NewEnumerator1
    Mass Mode=::NewEnumerator1
    Mesh Scale Mode=::NewEnumerator3
    Mesh Uniform Scale=&Multiply_Float(A=$System.SysParticleScale, B=&RandomRangeFloat(Minimum=0.1, Maximum=0.4))
    Saturation Range:Vector2f=<0.6 1.2>
    UsePositionOffset=false
    Value Range:Vector2f=<0.8 1.2>
    Write Material Random=true
   ) {module=InitializeParticle001}
   &ShapeLocation(
    Apply Owner Scale=HlslExpression() {hlslExpression="abs(Engine.Owner.Scale)"}
    MaskSpawnGroup=false
    Non Uniform Scale=$System.SysObjectScale
    Offset Mode=::NewEnumerator0
    Override Seed=false
    UseOwnerScale=false
   ) {module=ShapeLocation002}
   &AddVelocity(
    Module.Use Falloff Distance=false
    Origin Offset=<0 0 -100>
    Origin Offset Coordinate Space=::World
    Override Seed=false
    Velocity Mode=::NewEnumerator1
    Velocity Speed=&RandomRangeFloat(Minimum=100, Maximum=600)
   )
   &InitialMeshOrientation(Use Rotation Vector=false)
   &MeshRotationForce(Rotation=&RandomRangeVector(Minimum=<-50 -50 -50>, Maximum=<50 50 50>))
   &ApplyInitialForces()
  ]
  particleUpdateScript [
   &ParticleState()
   &ScaleMeshSize(Scale Factor=&Multiply_VectorByFloat(
     Vector=<1 1 1>
     Float=&FloatFromCurve(
      FloatCurve=DataInterface() {propertyValues {Curve {keys [
          {arriveTangent=13.7848, interpMode=RCIM_Cubic, leaveTangent=13.7848, tangentMode=RCTM_User, time=0, value=0}
          <at 0.15 1>
          {interpMode=RCIM_Cubic, tangentMode=RCTM_User, time=0.8, value=1}
          {tangentMode=RCTM_User, time=1, value=0}
         ]}}}
      Scale Curve=0.75
     )
    ))
   &GravityForce(Gravity=&Multiply_VectorByFloat(Vector=<0 0 -980>, Float=1))
   &Drag(Drag=0, Ignore Mass=false, Rotational Drag=3, Use Linear Drag=true, Use Rotational Drag=true)
   &Collision(
    Age Colliding Particles=false
    Collision Enabled=&SetBoolByFloatComparison(A=$Particles.Lifetime, B=0.2)
    Control Roll On Collision=true
    Friction=0.6
    Friction During a Bounce=0.1
    GPU Collision Type=::NewEnumerator2
    Kill Particles Lodged Within Meshes=false
    KillOnCollision=false
    Mesh Dimensions:Vector3f=$Particles.Scale
    Module.Control Rest State Via Surface Normal=false
    Module.EnableMaxCollisionCount=false
    Module.Randomize Collision Normal Vector=false
    Radius Calculation Type=::NewEnumerator1
    Radius Scale for Rotation=30
    Restitution=0.3
    ScaleRateOfRotation=true
   ) {module=Collision001}
   &SolveForcesAndVelocity(Clamp Velocity=false, Limit Acceleration=false)
   &KillParticles(Kill Particles=HlslExpression() {hlslExpression="(Particles.Position.z < System.CollisionPosWorld.z) &&&& System.CollisionValid"})
   /SetVariables(Particles.Scale=HlslExpression() {hlslExpression="abs(Particles.Scale)"}) {isSetParameters=true}
  ]
  rendererClasses [&NiagaraMeshRendererProperties]
  renderers [{rendererClass=&NiagaraMeshRendererProperties, rendererIndex=0}]
 }
 GPUDroplets {
  simTarget=GPUComputeSim
  emitterUpdateScript [
   &EmitterState()
   &SpawnBurst_Instantaneous(Spawn Count=16, Use Spawn Probability=false, UseLoopCountLimit=false)
  ]
  particleSpawnScript [
   &InitializeParticle(
    Color=&RandomRangeLinearColor(Minimum=<rgba 0.0983 0.423 1 0.35>, Maximum=<rgba 0.7448 0.8834 1 0.5>)
    Lifetime Max=0.1
    Lifetime Min=0.265
    Lifetime Mode=::NewEnumerator1
    Sprite Size Mode=::NewEnumerator1
    Uniform Sprite Size Max=5
    Uniform Sprite Size Min=4
    UsePositionOffset=false
    Write Material Random=true
   )
   &ShapeLocation(
    Apply Owner Scale=HlslExpression() {hlslExpression="abs(Engine.Owner.Scale)"}
    MaskSpawnGroup=false
    Non Uniform Scale=$System.SysLocationScale
    Offset=$System.SysLocOffset
    Offset Mode=::NewEnumerator0
    Override Seed=false
    Sphere Radius=75
    Sphere Surface Distribution=0.4286
    UseOwnerScale=false
   )
   &AddVelocity(
    Module.Use Falloff Distance=false
    Origin Offset=$System.SysLocOffset
    Override Seed=false
    Velocity Mode=::NewEnumerator1
    Velocity Speed=&Multiply_Float(A=&RandomRangeFloat(Minimum=100, Maximum=250), B=$System.SysParticleScale)
   )
  ]
  particleUpdateScript [
   &Drag(Drag=0.25, Use Linear Drag=true, Use Rotational Drag=true)
   &SolveForcesAndVelocity(Clamp Velocity=false, Limit Acceleration=false)
   &ParticleState()
   &ScaleSpriteSize(Scale Sprite Size Mode=::NewEnumerator0, Uniform Scale Factor=$System.SysParticleScale) {module=ScaleSpriteSize001}
   &ScaleSpriteSizeBySpeed(Max Scale Factor=<1 3>, Velocity Threshold=300)
   &ScaleSpriteSize(Uniform Curve Sprite Scale=DataInterface() {propertyValues {Curve {keys [
        {arriveTangent=4.1785, interpMode=RCIM_Cubic, leaveTangent=4.1785, tangentMode=RCTM_User, time=0, value=0}
        {arriveTangent=1.1179, interpMode=RCIM_Cubic, leaveTangent=1.1179, tangentMode=RCTM_User, time=0.2042, value=0.6228}
        {arriveTangent=0.0224, interpMode=RCIM_Cubic, leaveTangent=0.0224, tangentMode=RCTM_User, time=1, value=1}
       ]}}})
   &ScaleColor(
    Scale Alpha=&FloatFromCurve(FloatCurve=DataInterface() {propertyValues {Curve {keys [
         {arriveTangent=9.0725, interpMode=RCIM_Cubic, leaveTangent=9.0725, tangentMode=RCTM_User, time=0, value=0}
         {arriveTangent=-0.3377, interpMode=RCIM_Cubic, leaveTangent=-0.3377, tangentMode=RCTM_User, time=0.1408, value=0.9847}
         {arriveTangent=-0.0405, interpMode=RCIM_Cubic, leaveTangent=-0.0405, tangentMode=RCTM_User, time=1, value=0}
        ]}}})
    Scale RGB=&Multiply_VectorByFloat(Vector=<1 1 1>, Float=1)
    ScaleA=true
    ScaleRGB=true
   )
   /SetVariables(Particles.SpriteFacing=&MakeCustomVectorFromBool(
     Bool=&SetBoolByFloatComparison(A=$System.OwnerScaleX, B=0)
     True Vector=&ReturnCameraVector(Camera Vectors=::NewEnumerator4, Coordinate Space=::Simulation, Origin Position=$Particles.Position)
     False Vector=&ReturnCameraVector(Camera Vectors=::NewEnumerator1, Coordinate Space=::Simulation)
    )) {isSetParameters=true}
  ]
  rendererClasses [&NiagaraSpriteRendererProperties]
  renderers [{rendererClass=&NiagaraSpriteRendererProperties}]
 }
 Ground_Position_Debug {
  bEnabled=false
  emitterUpdateScript [
   &EmitterState()
   &SpawnBurst_Instantaneous(Use Spawn Probability=false, UseLoopCountLimit=false)
  ]
  particleSpawnScript [
   &InitializeParticle(
    Color=<rgba 0.8751 0 1 1>
    Lifetime=20
    Position=$System.CollisionPosWorld
    Position Mode=::NewEnumerator1
    Sprite Size Mode=::NewEnumerator3
    Uniform Sprite Size=200
    UsePositionOffset=false
    Write Material Random=true
   )
   &RegionMask(Region Type:ENiagaraDebugDrawShape=::NewEnumerator2, UseOffset=false, UseRotation=false)
  ]
  particleUpdateScript [&ParticleState()]
  rendererClasses [&NiagaraSpriteRendererProperties]
  renderers [{rendererClass=&NiagaraSpriteRendererProperties}]
 }
 Ice_Sprites_HIGH {
  simTarget=GPUComputeSim
  emitterUpdateScript [
   &EmitterState()
   &SpawnBurst_Instantaneous(
    Spawn Count=&MakeIntFromFloat(Float=&FloatFromCurve(
      FloatCurve=DataInterface() {propertyValues {Curve {keys [<at 2 30>,<at 7 100>,<at 19 200>]}}}
      CurveIndex=$System.SysSpawnCount
      Scale Curve=3
     ))
    Use Spawn Probability=false
    UseLoopCountLimit=false
   ) {module=SpawnBurst_Instantaneous002}
  ]
  particleSpawnScript [
   &InitializeParticle(
    AdjustAlpha=true
    AdjustHue=true
    AdjustSaturation=true
    AdjustValue=true
    Alpha Scale Range:Vector2f=<1 1>
    Color=$System.IceColorTOD
    Color Mode=::NewEnumerator3
    Hue Shift Range:Vector2f=<-0.03 0.04>
    Lifetime Max=1.5
    Lifetime Min=0.9
    Lifetime Mode=::NewEnumerator1
    Mass Mode=::NewEnumerator1
    Saturation Range:Vector2f=<0.5 1.2>
    Sprite Rotation Mode=::NewEnumerator2
    Sprite Size Mode=::NewEnumerator1
    Uniform Sprite Size Max=40
    Uniform Sprite Size Min=15
    UsePositionOffset=false
    Value Range:Vector2f=<0.6 3>
    Write Material Random=true
   )
   &ShapeLocation(
    Apply Owner Scale=HlslExpression() {hlslExpression="abs(Engine.Owner.Scale)"}
    MaskSpawnGroup=false
    Non Uniform Scale=$System.SysObjectScale
    Offset Mode=::NewEnumerator0
    Override Seed=false
    UseOwnerScale=false
   ) {module=ShapeLocation002}
   &AddVelocity(
    Module.Use Falloff Distance=false
    Origin Offset=<0 0 -200>
    Origin Offset Coordinate Space=::World
    Override Seed=false
    Velocity Mode=::NewEnumerator1
    Velocity Speed=&RandomRangeFloat(Minimum=100, Maximum=800)
   ) {module=AddVelocity001}
  ]
  particleUpdateScript [
   &GravityForce()
   &Drag(Drag=0.25, Use Linear Drag=true, Use Rotational Drag=true)
   &ParticleState()
   &SpriteRotationRate(Rotation Rate=&RandomRangeFloat(Minimum=800, Maximum=&RandomRangeFloat(Minimum=-1, Maximum=1)))
   &ScaleSpriteSize(Uniform Curve Sprite Scale=DataInterface() {propertyValues {Curve {keys [
        {arriveTangent=9.0725, interpMode=RCIM_Cubic, leaveTangent=9.0725, tangentMode=RCTM_User, time=0, value=0}
        {arriveTangent=-2.5249, interpMode=RCIM_Cubic, leaveTangent=-2.5249, tangentMode=RCTM_User, time=0.1408, value=0.9847}
        {arriveTangent=-0.0405, interpMode=RCIM_Cubic, leaveTangent=-0.0405, tangentMode=RCTM_User, time=1, value=0}
       ]}}})
   &SolveForcesAndVelocity(Clamp Velocity=false, Limit Acceleration=false)
   &ScaleColor(Scale RGB=&Multiply_VectorByFloat(Vector=<1 1 1>, Float=0.5), ScaleA=true, ScaleRGB=true)
   &CameraOffset(Camera Offset Amount=-200)
   &KillParticles(Kill Particles=HlslExpression() {hlslExpression="(Particles.Position.z < System.CollisionPosWorld.z) &&&& System.CollisionValid"})
   /SetVariables(Particles.SpriteFacing=&MakeCustomVectorFromBool(
     Bool=&SetBoolByFloatComparison(A=$System.OwnerScaleX, B=0)
     True Vector=&ReturnCameraVector(Camera Vectors=::NewEnumerator4, Coordinate Space=::Simulation, Origin Position=$Particles.Position)
     False Vector=&ReturnCameraVector(Camera Vectors=::NewEnumerator1, Coordinate Space=::Simulation)
    )) {isSetParameters=true}
  ]
  rendererClasses [&NiagaraSpriteRendererProperties]
  renderers [{rendererClass=&NiagaraSpriteRendererProperties}]
 }
 Ice_Sprites_LOW {rendererClasses [], renderers []}
 Mist {
  emitterUpdateScript [
   &EmitterState()
   &SpawnBurst_Instantaneous(
    Spawn Count=&MakeIntFromFloat(Float=&FloatFromCurve(FloatCurve=DataInterface() {propertyValues {Curve {keys [<at 2 4>,<at 7 12>,<at 19 20>]}}}, CurveIndex=$System.SysSpawnCount))
    Use Spawn Probability=false
    UseLoopCountLimit=false
   ) {module=SpawnBurst_Instantaneous002}
  ]
  particleSpawnScript [
   &InitializeParticle(
    AdjustAlpha=false
    AdjustHue=true
    AdjustSaturation=true
    AdjustValue=true
    Color=<rgba 0.6 0.6765 1 0.4>
    Color Mode=::NewEnumerator3
    Hue Shift Range:Vector2f=<0 0.1>
    Lifetime Max=0.75
    Lifetime Min=0.5
    Lifetime Mode=::NewEnumerator1
    Saturation Range:Vector2f=<0.8 1.2>
    Sprite Rotation Mode=::NewEnumerator2
    Sprite Size Mode=::NewEnumerator3
    Uniform Sprite Size=&Multiply_Float(A=80, B=$System.SysParticleScale)
    UsePositionOffset=false
    Write Material Random=true
   )
   &CameraOffset(Camera Offset Amount=35)
   &ShapeLocation(
    Apply Owner Scale=HlslExpression() {hlslExpression="abs(Engine.Owner.Scale)"}
    MaskSpawnGroup=false
    Non Uniform Scale=$System.SysObjectScale
    Offset Mode=::NewEnumerator0
    Override Seed=false
    UseOwnerScale=false
   ) {module=ShapeLocation002}
   &AddVelocity(
    Module.Use Falloff Distance=false
    Origin Offset=<0 0 -200>
    Origin Offset Coordinate Space=::World
    Override Seed=false
    Velocity Mode=::NewEnumerator1
    Velocity Speed=&RandomRangeFloat(Minimum=50, Maximum=300)
   ) {module=AddVelocity001}
  ]
  particleUpdateScript [
   &ParticleState()
   &ScaleSpriteSize(Uniform Curve Sprite Scale=DataInterface() {propertyValues {Curve {keys [
        {arriveTangent=4.1785, interpMode=RCIM_Cubic, leaveTangent=4.1785, tangentMode=RCTM_User, time=0, value=0}
        {arriveTangent=1.1179, interpMode=RCIM_Cubic, leaveTangent=1.1179, tangentMode=RCTM_User, time=0.2042, value=0.6228}
        {arriveTangent=0.0224, interpMode=RCIM_Cubic, leaveTangent=0.0224, tangentMode=RCTM_User, time=1, value=1}
       ]}}})
   &SubUVAnimation(
    "Sprite Renderer "=DataInterface() {propertyValues {SpriteRenderer=&self:Mist_0.NiagaraSpriteRendererProperties_0}}
    SubUV Animation Mode:ENiagara_SubUVLookupModeV2=::NewEnumerator1
    UseEndFrame=false
    UseStartFrame=false
   )
   &ScaleColor(
    Scale Alpha=&FloatFromCurve(FloatCurve=DataInterface() {propertyValues {Curve {keys [
         {arriveTangent=11.3101, interpMode=RCIM_Cubic, leaveTangent=11.3101, tangentMode=RCTM_User, time=0, value=0}
         {arriveTangent=0.0252, interpMode=RCIM_Cubic, leaveTangent=0.0252, tangentMode=RCTM_User, time=0.2, value=0.7}
         {arriveTangent=-0.0405, interpMode=RCIM_Cubic, leaveTangent=-0.0405, tangentMode=RCTM_User, time=1, value=0}
        ]}}})
    Scale RGB=&MakeVectorFromLinearColorRGB(Color=&FNTimeOfDayColor(
      Night Color=<rgba 0.7604 0.944 1 1>
      Morning Color=<rgba 0.7396 0.9543 1 1>
      Day Color=<rgba 0.7604 0.944 1 1>
      Evening Color=<rgba 0.7604 0.944 1 1>
     ))
    ScaleA=true
    ScaleRGB=false
   )
   &AccelerationForce(Acceleration=<0 0 -120>)
   &SolveForcesAndVelocity(Clamp Velocity=false, Limit Acceleration=false)
   &DynamicMaterialParameters(
    Index 0 Param 1=3
    Index 0 Param 4=1
    Module.Index0Float0Write=true
    Module.Index0Float1Write=true
    Module.Index0Float2Write=true
    Module.Index0Float3Write=true
   )
   &KillParticles(Kill Particles=HlslExpression() {hlslExpression="(Particles.Position.z < System.CollisionPosWorld.z) &&&& System.CollisionValid"})
   /SetVariables(Particles.SpriteFacing=&MakeCustomVectorFromBool(
     Bool=&SetBoolByFloatComparison(A=$System.OwnerScaleX, B=0)
     True Vector=&ReturnCameraVector(Camera Vectors=::NewEnumerator4, Coordinate Space=::Simulation, Origin Position=$Particles.Position)
     False Vector=&ReturnCameraVector(Camera Vectors=::NewEnumerator1, Coordinate Space=::Simulation)
    )) {isSetParameters=true}
  ]
  rendererClasses [&NiagaraSpriteRendererProperties]
  renderers [{rendererClass=&NiagaraSpriteRendererProperties}]
 }
}
summary {
 systemName=NS_IceDestruction
 emitters [
  {emitterName=Mist, rendererClasses [&NiagaraSpriteRendererProperties]}
  {emitterName=GPUDroplets, simTarget=GPUComputeSim, rendererClasses [&NiagaraSpriteRendererProperties]}
  {emitterName=Ice_Sprites_HIGH, simTarget=GPUComputeSim, rendererClasses [&NiagaraSpriteRendererProperties]}
  {emitterName=Ice_Sprites_LOW, rendererClasses []}
  {bEnabled=false, emitterName=Ground_Position_Debug, rendererClasses [&NiagaraSpriteRendererProperties]}
  {bEnabled=false, emitterName=Bounds_Debug, simTarget=GPUComputeSim, rendererClasses [&NiagaraSpriteRendererProperties]}
  {emitterName=Chunks_CPU_LOW, rendererClasses [&NiagaraMeshRendererProperties]}
  {emitterName=Chunks_GPU_HIGH, simTarget=GPUComputeSim, rendererClasses [&NiagaraMeshRendererProperties]}
 ]
 userVariables [
  User.LocationOffset {type=Vector3f, description="", defaultValue {struct=Vector3f, value=<0 0 0>}}
  User.LocationScale {type=Vector3f, description="", defaultValue {struct=Vector3f, value=<3 3 3>}}
  User.ParticleScale {type=NiagaraFloat, defaultValue=1, description=""}
 ]
}
